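# Fortnite conversation option brief widget for a supply drop radio choice, showing the drop's icon, name, description, rarity and remaining stock.
# OnConfigure reads the drop data from the considered conversation task, tints background and glow materials by rarity, and switches between select and out-of-stock states.
# A delegate on the game state's supply drop radio component, bound on Construct and unbound on Destruct, refreshes stock or shows out-of-stock when replicated data changes.
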WidgetBlueprint /SupplyDropRadioGameplay/UI/SupplyDropRadioOptionBrief
graphs [
 EventGraph {nodes [
   n1 = ColorizeRarity!() -> (OutputDelegate, then, Rarity:EFortRarity Enum=Common) {is=title K2Node_CustomEvent}
   > n2 = GetRarityData!(execute, Rarity:EFortRarity Enum=@n1.Rarity) -> (then, ReturnValue:Fort Rarity Item Data Structure)
   > n3 = GetDynamicMaterial(self=Get Image_ItemBG() -> (Image_ItemBG:Image Object Reference))
   n4 = Break Fort Rarity Item Data!(FortRarityItemData:"Fort Rarity Item Data Structure (by ref)"=@n2) -> (
    Name:Text
    Color1:Linear Color Structure
    Color2:Linear Color Structure
    Color3:Linear Color Structure
    Color4:Linear Color Structure
    Color5:Linear Color Structure
    Radius:float=~
    Falloff:float=~
    Brightness:float=~
    Roughness:float=~
    Glow:float=~
   ) {is=title K2Node_BreakStruct}
   > SetVectorParameterValue(self=@n3, ParameterName=BGColor, Value=@n4.Color3)
   > SetVectorParameterValue(self=@n3, ParameterName=GradientColor, Value=@n4.Color2)
   > SetVectorParameterValue(self=@n3, ParameterName=OutlineColor, Value=@n4.Color1)
   > n5 = GetDynamicMaterial(self=Get Image_RarityGlow() -> (Image_RarityGlow:Image Object Reference))
   > SetVectorParameterValue(self=@n5, ParameterName=Color1, Value=@n4.Color1)
   > SetVectorParameterValue(self=@n5, ParameterName=Color2, Value=@n4.Color3)
   HandleSupplyDropRadioRepDataUpdated() {is=title K2Node_CustomEvent}
   > n6 = Get!(execute:Exec) -> (then:Exec, else:Exec, ConvoTask:Convo Task Supply Drop Radio Object Reference)
   > n7 = Get!(execute) -> (then, else, SupplyDropRadioGameStateComponent:BP Supply Drop Radio Game State Component Object Reference)
   n8 = IsStockAvailableForConversationIndex!(
    self:BP Supply Drop Radio Game State Component Object Reference=@n7
    ConversationIndex:int=GetConversationIndex!(self:Convo Task Supply Drop Radio Object Reference=@n6) -> (Index:int=0)
   ) -> (bAvailable:bool, Count:int=0)
   > n9 = Branch(Condition=@n8)
   > n10 = GetSupplyDropRadioData!(execute, self:Convo Task Supply Drop Radio Object Reference=@n6, WorldContextObject:Object Reference=@n7) -> (then, bValid:bool, Data:Supply Drop Radio Data Struct Structure)
   > UpdateStockWidget!(
    execute
    self
    AvailableValue:int=@n8.Count
    MaxStockValue:int=Break Supply Drop Radio Data Struct!(SupplyDropRadioDataStruct:"Supply Drop Radio Data Struct Structure (by ref)"=@n10.Data) -> (
     bEnabled_35_#1:bool=~
     PlaylistTagsQuery_39_#2:Gameplay Tag Query Structure
     GameplayTag_5_#3:Gameplay Tag Structure
     Stock_34_#4:int
     Rarity_12_#5:EFortRarity Enum
     Name_15_#6:Text
     Description_17_#7:Text
     SupplyDropRadioActorClass_58_#8:Building Gameplay Actor Supply Drop Radio Soft Class Reference
     PreviewImageSoftObj_24_#9:Texture 2D Soft Object Reference
     ChoiceImageSoftObj_29_#10:Texture 2D Soft Object Reference
     QuestTags_44_#11:Gameplay Tag Container Structure
     FallHeight_47_#12:double=~
     GravityMultiplier_PreOpen_49_#13:double=~
     GravityMultiplier_PostOpen_51_#14:double=~
     BalloonOpenedFallSpeed_53_#15:double=~
     DistanceFromOriginalSpawnLocForBalloonOpen_56_#16:double=~
    ) {is=title K2Node_BreakStruct}.Stock_34_#4
   ) -> (then)
   HandleOutOfStock(execute=@n9.else)
   Event Destruct()
   > Parent: Destruct()
   > n11 = Get()
   > Unbind Event from On Supply Drop Radio Rep Data Updated(self:BP Supply Drop Radio Game State Component Object Reference=@n11, Delegate=Create Event()) {is=title K2Node_RemoveDelegate}
   K2Node_Event_1 = HandleOutOfStock() {is=title K2Node_CustomEvent}
   > SetActiveWidget(
    self=Get StateVisibilitySwitcher() -> (StateVisibilitySwitcher:Common Visibility Switcher Object Reference)
    Widget=Get CannotBuy() -> (CannotBuy:Overlay Object Reference)
   )
   n12 = Event OnConfigure!() -> (
    OutputDelegate
    then
    ClientContext:"Conversation Context Structure (by ref)"
    OptionEntry:"Client Conversation Option Entry Structure (by ref)"
    SelectedIndex:int=0
   ) {is=title K2Node_Event}
   > n13 = GetParticipantPlayerController!(execute, Context:"Conversation Context Structure (by ref)"=@n12.ClientContext) -> (then, ReturnValue:Fort Player Controller Object Reference)
   > n14 = GetTaskBeingConsidered!(execute, Context:"Conversation Context Structure (by ref)"=@n12.ClientContext) -> (then, ReturnValue:Conversation Task Node Object Reference)
   > n15 = Cast To ConvoTask_SupplyDropRadio!(execute, Object=@n14) -> (then, CastFailed, AsConvo Task Supply Drop Radio:Convo Task Supply Drop Radio Object Reference) {is=title K2Node_DynamicCast}
   > n16 = Set ConvoTask(ConvoTask:Convo Task Supply Drop Radio Object Reference=@n15) -> (Output_Get:Convo Task Supply Drop Radio Object Reference)
   > SetSupplyDropInfo!(execute, self, WorldContextObject:Fort Player Controller Object Reference=@n13) -> (then)
   > n17 = Branch(Condition=IsStockAvailable!(self:Convo Task Supply Drop Radio Object Reference=@n16, WorldContextObject:Object Reference=@n13) -> (bAvailable:bool, Count:int=0))
   > SetActiveWidget(self=Get StateVisibilitySwitcher(), Widget=Get CanBuy() -> (CanBuy:Overlay Object Reference))
   HandleOutOfStock(execute=@n17.else) {is=title K2Node_CallFunction}
   Event Construct()
   > Parent: Construct()
   n18 = GetGameStateAthena!() -> (ReturnValue:Fort Game State Athena Object Reference)
   > Is Valid(InputObject=@n18)
   n19 = GetComponentByClass!(
    self:Actor Object Reference=@n18
    ComponentClass:Actor Component Class Reference=/SupplyDropRadioGameplay/Gameplay/BP_SupplyDropRadioGameStateComponent.BP_SupplyDropRadioGameStateComponent_C
   ) -> (ReturnValue:BP Supply Drop Radio Game State Component Object Reference)
   > Is Valid(InputObject=@n19)
   > n20 = Set SupplyDropRadioGameStateComponent(SupplyDropRadioGameStateComponent:BP Supply Drop Radio Game State Component Object Reference=@n19) -> (Output_Get:BP Supply Drop Radio Game State Component Object Reference)
   > Bind Event to On Supply Drop Radio Rep Data Updated(self:BP Supply Drop Radio Game State Component Object Reference=@n20, Delegate=Create Event()) {is=title K2Node_AddDelegate}
  ]}
 SetSupplyDropInfo {nodes [
   n1 = SetSupplyDropInfo!() -> (then, WorldContextObject:Fort Player Controller Object Reference) {is=title K2Node_FunctionEntry}
   > n2 = GetSupplyDropRadioData(self=Get ConvoTask() -> (ConvoTask:Convo Task Supply Drop Radio Object Reference), WorldContextObject=@n1)
   > Branch(Condition=@n2)
   n3 = Break Supply Drop Radio Data Struct(SupplyDropRadioDataStruct=@n2.Data) -> (
    bEnabled_35_#1=~
    FallHeight_47_#12=~
    GravityMultiplier_PreOpen_49_#13=~
    GravityMultiplier_PostOpen_51_#14=~
    BalloonOpenedFallSpeed_53_#15=~
    DistanceFromOriginalSpawnLocForBalloonOpen_56_#16=~
   )
   > ColorizeRarity!(execute, self, Rarity:EFortRarity Enum=@n3.Rarity_12_#5) -> (then) {is=title K2Node_CallFunction}
   > "SetText (Text)"(self=Get Text_Title() -> (Text_Title:Common Text Object Reference), InText=@n3.Name_15_#6)
   > "SetText (Text)"(self=Get Text_Description() -> (Text_Description:Common Text Object Reference), InText=@n3.Description_17_#7)
   > SetBrushFromSoftTexture!(
    execute
    self:Image Object Reference=Get Image_Icon() -> (Image_Icon:Image Object Reference)
    SoftTexture:Texture 2D Soft Object Reference=@n3.PreviewImageSoftObj_24_#9
    bMatchSize:bool=true
   ) -> (then)
   > UpdateStockWidget(
    AvailableValue=IsStockAvailable(self=Get ConvoTask(), WorldContextObject=Get WorldContextObject() -> (WorldContextObject:Fort Player Controller Object Reference)) -> (Count=0).Count
    MaxStockValue=Break Supply Drop Radio Data Struct(SupplyDropRadioDataStruct=@n2.Data) -> (
     bEnabled_35_#1=~
     FallHeight_47_#12=~
     GravityMultiplier_PreOpen_49_#13=~
     GravityMultiplier_PostOpen_51_#14=~
     BalloonOpenedFallSpeed_53_#15=~
     DistanceFromOriginalSpawnLocForBalloonOpen_56_#16=~
    ).Stock_34_#4
   )
  ]}
 UpdateStockWidget {nodes [
   n1 = UpdateStockWidget!() -> (then, AvailableValue:int, MaxStockValue:int) {is=title K2Node_FunctionEntry}
   > n2 = Branch(Condition="integer < integer"(A=@n1))
   > SetVisibility(self=Get StockRemaining() -> (StockRemaining:Common Text Object Reference), InVisibility=Collapsed)
   SetVisibility(execute=@n2.else, self=Get StockRemaining(), InVisibility=SelfHitTestInvisible)
   > "SetText (Text)"(
    self=Get StockRemaining()
    InText=Format Text!(Format="LOCTEXT(\"Stock: {available}/{max}\")", available:int=@n1, max:int=@n1.MaxStockValue) -> (Result)
   )
  ]}
]
variables [
 ConvoTask {type=ConvoTask_SupplyDropRadio_C, default=null}
 DebugItem {type=FortItem, default=null}
 SupplyDropRadioData {
  type=SupplyDropRadioDataStruct
  default {
   name=Name
   bEnabled=false
   balloonOpenedFallSpeed=550
   choiceImageSoftObj=null
   description=Description
   distanceFromOriginalSpawnLocForBalloonOpen=16000
   fallHeight=30000
   gravityMultiplier_PostOpen=0.35
   gravityMultiplier_PreOpen=0.175
   previewImageSoftObj=null
   rarity=Epic
   stock=0
   supplyDropRadioActorClass=null
   gameplayTag {tagName=None}
   playlistTagsQuery {autoDescription="", tokenStreamVersion=0, userDescription="", queryTokenStream [], tagDictionary []}
   questTags {gameplayTags []}
  }
 }
 SupplyDropRadioGameStateComponent {type=BP_SupplyDropRadioGameStateComponent_C, default=null}
]
widgets {
 info {
  inheritedWidgetCount=0
  parentClass=/Script/FortniteConversationUI.FortConversationOptionBrief
  rootWidgetClass=&SizeBox
  widgetCount=38
 }
 widgets [
  SizeBox_36 {
   widgetClassPath=&SizeBox
   widget=SizeBox_36
   properties {heightOverride=350, maxDesiredWidth=768, minDesiredWidth=384, widthOverride=350}
  }
  Overlay_2 {widgetClassPath=&Overlay, parent=SizeBox_36, widget=Overlay_2, slot {class=SizeBoxSlot}}
  ItemInfo {widgetClassPath=&Overlay, bIsVariable=true, parent=Overlay_2, widget=ItemInfo, slot {class=OverlaySlot}}
  VerticalBox_2 {
   widgetClassPath=&VerticalBox
   parent=ItemInfo
   widget=VerticalBox_2
   slot {class=OverlaySlot, properties {horizontalAlignment=HAlign_Fill}}
  }
  Border_168 {
   widgetClassPath=&Border
   parent=VerticalBox_2
   widget=Border_168
   properties {brushColor=<rgba 0 0 0 0.446>, padding=<ltrb 4 4 4 4>}
   slot {class=VerticalBoxSlot}
  }
  Overlay_0 {widgetClassPath=&Overlay, parent=Border_168, widget=Overlay_0, slot {class=BorderSlot, properties {padding=<ltrb 4 4 4 4>}}}
  HorizontalBox_0 {widgetClassPath=&HorizontalBox, parent=Overlay_0, widget=HorizontalBox_0, slot {class=OverlaySlot}}
  Overlay_WeaponPreview {
   widgetClassPath=&Overlay
   parent=HorizontalBox_0
   widget=Overlay_WeaponPreview
   slot {class=HorizontalBoxSlot, properties {horizontalAlignment=HAlign_Left, verticalAlignment=VAlign_Top}}
  }
  Overlay_ItemBG {
   widgetClassPath=&Overlay
   parent=Overlay_WeaponPreview
   widget=Overlay_ItemBG
   slot {class=OverlaySlot, properties {verticalAlignment=VAlign_Center}}
  }
  Image_84 {
   widgetClassPath=&Image
   bIsVariable=true
   parent=Overlay_ItemBG
   widget=Image_84
   properties {colorAndOpacity=<rgba 0.0052 0.0052 0.0056 0.2>, renderTransform {angle=0, scale=<1 1>, shear=<0 0>, translation=<8 8>}}
   slot {class=OverlaySlot, properties {horizontalAlignment=HAlign_Fill, verticalAlignment=VAlign_Fill}}
  }
  Overlay_1 {
   widgetClassPath=&Overlay
   parent=Overlay_ItemBG
   widget=Overlay_1
   slot {class=OverlaySlot, properties {horizontalAlignment=HAlign_Fill, verticalAlignment=VAlign_Fill, padding=<ltrb 3 3 3 3>}}
  }
  Image_ItemBG {
   widgetClassPath=&Image
   bIsVariable=true
   parent=Overlay_1
   widget=Image_ItemBG
   properties {brush {
     bIsDynamicallyLoaded=false
     drawAs=Image
     imageType=NoImage
     mirroring=NoMirror
     resourceName=None
     resourceObject=/Game/Athena/Playlists/Goose/UI/Material/M_GooseItemBG_Inst
     tiling=NoTile
     imageSize=<90 90>
     margin=<ltrb 0 0 0 0>
     outlineSettings {
      bUseBrushTransparency=false
      roundingType=HalfHeightRadius
      width=0
      color {colorUseRule=UseColor_Specified, specifiedColor=<rgba 0 0 0 0>}
      cornerRadii=<0 0 0 0>
     }
     tintColor {colorUseRule=UseColor_Specified, specifiedColor=<rgba 1 1 1 1>}
     uVRegion {bIsValid=false, max=<0 0>, min=<0 0>}
    }}
   slot {class=OverlaySlot, properties {horizontalAlignment=HAlign_Fill, verticalAlignment=VAlign_Fill}}
  }
  Image_RarityGlow {
   widgetClassPath=&Image
   bIsVariable=true
   parent=Overlay_1
   widget=Image_RarityGlow
   properties {
    renderOpacity=0
    visibility=HitTestInvisible
    brush {
     bIsDynamicallyLoaded=false
     drawAs=Image
     imageType=NoImage
     mirroring=NoMirror
     resourceName=None
     resourceObject=/Game/Athena/HUD/Interaction/M_UI_RotatingBurst
     tiling=NoTile
     imageSize=<220 220>
     margin=<ltrb 0 0 0 0>
     outlineSettings {
      bUseBrushTransparency=false
      roundingType=HalfHeightRadius
      width=0
      color {colorUseRule=UseColor_Specified, specifiedColor=<rgba 0 0 0 0>}
      cornerRadii=<0 0 0 0>
     }
     tintColor {colorUseRule=UseColor_Specified, specifiedColor=<rgba 1 1 1 1>}
     uVRegion {bIsValid=false, max=<0 0>, min=<0 0>}
    }
   }
   slot {class=OverlaySlot, properties {horizontalAlignment=HAlign_Center, verticalAlignment=VAlign_Center, padding=<ltrb -200 -200 -200 -200>}}
  }
  SizeBox_0 {
   widgetClassPath=&SizeBox
   parent=Overlay_1
   widget=SizeBox_0
   properties {bOverride_HeightOverride=true, bOverride_WidthOverride=true, heightOverride=128, widthOverride=128}
   slot {class=OverlaySlot, properties {horizontalAlignment=HAlign_Fill, verticalAlignment=VAlign_Fill}}
  }
  Overlay_ItemIcon {widgetClassPath=&Overlay, bIsVariable=true, parent=SizeBox_0, widget=Overlay_ItemIcon, slot {class=SizeBoxSlot}}
  Image_IconShadow {
   widgetClassPath=&Image
   bIsVariable=true
   parent=Overlay_ItemIcon
   widget=Image_IconShadow
   properties {
    visibility=Collapsed
    brush {
     bIsDynamicallyLoaded=false
     drawAs=Image
     imageType=NoImage
     mirroring=NoMirror
     resourceName=None
     resourceObject=/Game/UI/Foundation/Textures/Icons/Weapons/Items/T-Icon-Weapons-SK-AutoHigh-Suppressed
     tiling=NoTile
     imageSize=<128 128>
     margin=<ltrb 0 0 0 0>
     outlineSettings {
      bUseBrushTransparency=false
      roundingType=HalfHeightRadius
      width=0
      color {colorUseRule=UseColor_Specified, specifiedColor=<rgba 0 0 0 0>}
      cornerRadii=<0 0 0 0>
     }
     tintColor {colorUseRule=UseColor_Specified, specifiedColor=<rgba 1 1 1 1>}
     uVRegion {bIsValid=false, max=<0 0>, min=<0 0>}
    }
    colorAndOpacity=<rgba 0.0052 0.0052 0.0056 0.3>
    renderTransform {angle=0, scale=<1 1>, shear=<0 0>, translation=<6 12>}
   }
   slot {class=OverlaySlot, properties {horizontalAlignment=HAlign_Fill, verticalAlignment=VAlign_Fill}}
  }
  Image_Icon {
   widgetClassPath=&Image
   bIsVariable=true
   parent=Overlay_ItemIcon
   widget=Image_Icon
   properties {brush {
     bIsDynamicallyLoaded=false
     drawAs=Image
     imageType=NoImage
     mirroring=NoMirror
     resourceName=None
     resourceObject=/Game/UI/Foundation/Textures/Icons/Athena/T-T-Icon-BR-SM-Athena-SupplyDrop
     tiling=NoTile
     imageSize=<32 32>
     margin=<ltrb 0 0 0 0>
     outlineSettings {
      bUseBrushTransparency=false
      roundingType=HalfHeightRadius
      width=0
      color {colorUseRule=UseColor_Specified, specifiedColor=<rgba 0 0 0 0>}
      cornerRadii=<0 0 0 0>
     }
     tintColor {colorUseRule=UseColor_Specified, specifiedColor=<rgba 1 1 1 1>}
     uVRegion {bIsValid=false, max=<0 0>, min=<0 0>}
    }}
   slot {class=OverlaySlot, properties {horizontalAlignment=HAlign_Fill, verticalAlignment=VAlign_Fill}}
  }
  VerticalBox_ItemText {
   widgetClassPath=&VerticalBox
   parent=HorizontalBox_0
   widget=VerticalBox_ItemText
   slot {class=HorizontalBoxSlot, properties {verticalAlignment=VAlign_Top, padding=<ltrb 10 0 0 0>}}
  }
  VerticalBox_1 {
   widgetClassPath=&VerticalBox
   parent=VerticalBox_ItemText
   widget=VerticalBox_1
   properties {renderTransform {angle=0, scale=<1 1>, shear=<0 0>, translation=<0 7>}}
   slot {class=VerticalBoxSlot, properties {horizontalAlignment=HAlign_Left, verticalAlignment=VAlign_Center, size {sizeRule=Fill, value=1}}}
  }
  SizeBox_2 {
   widgetClassPath=&SizeBox
   parent=VerticalBox_1
   widget=SizeBox_2
   properties {bOverride_MaxDesiredWidth=true, bOverride_MinDesiredWidth=true, maxDesiredWidth=512, minDesiredWidth=250, widthOverride=345.8244}
   slot {class=VerticalBoxSlot, properties {size {sizeRule=Fill, value=1}}}
  }
  Text_Title {
   widgetClassPath=&CommonTextBlock
   bIsVariable=true
   parent=SizeBox_2
   widget=Text_Title
   properties {
    text=Radio Drop
    textTransformPolicy=ToUpper
    wrapTextAt=512
    font {
     bForceMonospaced=false
     bMaterialIsStencil=false
     fontMaterial=null
     fontObject=&Burbank
     letterSpacing=0
     monospacedWidth=1
     size=18
     skewAmount=0.3
     typefaceFontName=BurbankRegular-Black
     outlineSettings {
      bApplyOutlineToDropShadows=true
      bMiteredCorners=false
      bSeparateFillAlpha=true
      outlineMaterial=&M-FontShadow
      outlineSize=3
      outlineColor=<rgba 0 0 0 0.5>
     }
    }
    shadowOffset=<0 0>
   }
   slot {class=SizeBoxSlot, properties {horizontalAlignment=HAlign_Left}}
  }
  VerticalBox {
   widgetClassPath=&VerticalBox
   parent=VerticalBox_ItemText
   widget=VerticalBox
   properties {renderTransform {angle=0, scale=<1 1>, shear=<0 0>, translation=<0 7>}}
   slot {class=VerticalBoxSlot, properties {horizontalAlignment=HAlign_Left, verticalAlignment=VAlign_Center, size {sizeRule=Fill, value=1}}}
  }
  SizeBox {
   widgetClassPath=&SizeBox
   parent=VerticalBox
   widget=SizeBox
   properties {bOverride_MaxDesiredWidth=true, bOverride_MinDesiredWidth=true, maxDesiredWidth=512, minDesiredWidth=160, widthOverride=345.8244}
   slot {class=VerticalBoxSlot, properties {size {sizeRule=Fill, value=1}}}
  }
  Text_Description {
   widgetClassPath=&CommonTextBlock
   bIsVariable=true
   parent=SizeBox
   widget=Text_Description
   properties {
    scrollStyle=&TextScrollStyle-Base_C
    text=Health supplies
    textTransformPolicy=ToUpper
    wrapTextAt=512
    font {
     bForceMonospaced=false
     bMaterialIsStencil=false
     fontMaterial=null
     fontObject=&BurbankSmall
     letterSpacing=0
     monospacedWidth=1
     size=16
     skewAmount=0.3
     typefaceFontName=Bold
     outlineSettings {
      bApplyOutlineToDropShadows=true
      bMiteredCorners=false
      bSeparateFillAlpha=true
      outlineMaterial=&M-FontShadow
      outlineSize=3
      outlineColor=<rgba 0 0 0 0.5>
     }
    }
    shadowOffset=<0 0>
   }
   slot {class=SizeBoxSlot, properties {horizontalAlignment=HAlign_Left}}
  }
  CommonTextBlock {
   widgetClassPath=&CommonTextBlock
   parent=Overlay_0
   widget=CommonTextBlock
   properties {
    text=Buy
    textTransformPolicy=ToUpper
    visibility=Collapsed
    font {
     bForceMonospaced=false
     bMaterialIsStencil=false
     fontMaterial=null
     fontObject=&Burbank
     letterSpacing=0
     monospacedWidth=1
     size=18
     skewAmount=0
     typefaceFontName=BurbankRegular-Black
     outlineSettings {
      bApplyOutlineToDropShadows=false
      bMiteredCorners=false
      bSeparateFillAlpha=false
      outlineMaterial=null
      outlineSize=2
      outlineColor=<rgba 0.0052 0.0052 0.0056 0.6>
     }
    }
    renderTransform {angle=0, scale=<1 1>, shear=<-8 0>, translation=<0 -20>}
    shadowOffset=<0 0>
   }
   slot {class=OverlaySlot, properties {verticalAlignment=VAlign_Fill}}
  }
  Border_SubDescription {
   widgetClassPath=&Border
   bIsVariable=true
   parent=VerticalBox_2
   widget=Border_SubDescription
   properties {verticalAlignment=VAlign_Top, visibility=HitTestInvisible, brushColor=<rgba 0.0273 0.0273 0.0273 0.8>, padding=<ltrb 0 0 0 0>}
   slot {class=VerticalBoxSlot, properties {verticalAlignment=VAlign_Top}}
  }
  StateVisibilitySwitcher {
   widgetClassPath=&CommonVisibilitySwitcher
   bIsVariable=true
   parent=Border_SubDescription
   widget=StateVisibilitySwitcher
   slot {class=BorderSlot, properties {verticalAlignment=VAlign_Top, padding=<ltrb 0 0 0 0>}}
  }
  CanBuy {
   widgetClassPath=&Overlay
   bIsVariable=true
   parent=StateVisibilitySwitcher
   widget=CanBuy
   slot {class=CommonVisibilitySwitcherSlot}
  }
  HorizontalBox_2 {
   widgetClassPath=&HorizontalBox
   parent=CanBuy
   widget=HorizontalBox_2
   slot {class=OverlaySlot, properties {horizontalAlignment=HAlign_Fill, verticalAlignment=VAlign_Fill, padding=<ltrb 8 8 12 8>}}
  }
  InteractionKeybindBox {
   widgetClassPath=&HorizontalBox
   parent=HorizontalBox_2
   widget=InteractionKeybindBox
   properties {visibility=HitTestInvisible}
   slot {class=HorizontalBoxSlot}
  }
  TextUnderKeybindText {
   widgetClassPath=&CommonTextBlock
   bIsVariable=true
   parent=InteractionKeybindBox
   widget=TextUnderKeybindText
   properties {
    justification=Center
    text=SELECT
    textTransformPolicy=ToUpper
    font {
     bForceMonospaced=false
     bMaterialIsStencil=false
     fontMaterial=null
     fontObject=&BurbankSmall
     letterSpacing=0
     monospacedWidth=1
     size=20
     skewAmount=0.3
     typefaceFontName=Black
     outlineSettings {
      bApplyOutlineToDropShadows=false
      bMiteredCorners=false
      bSeparateFillAlpha=true
      outlineMaterial=&M-FontShadow
      outlineSize=2
      outlineColor=<rgba 0 0 0 0.5>
     }
    }
    shadowOffset=<0 0>
   }
   slot {class=HorizontalBoxSlot, properties {verticalAlignment=VAlign_Center, padding=<ltrb 5 0 0 0>}}
  }
  VerticalBox_0 {
   widgetClassPath=&VerticalBox
   parent=HorizontalBox_2
   widget=VerticalBox_0
   slot {class=HorizontalBoxSlot, properties {horizontalAlignment=HAlign_Right, verticalAlignment=VAlign_Center, size {sizeRule=Fill, value=1}}}
  }
  SizeBox_1 {widgetClassPath=&SizeBox, parent=VerticalBox_0, widget=SizeBox_1, slot {class=VerticalBoxSlot}}
  StockRemaining {
   widgetClassPath=&CommonTextBlock
   bIsVariable=true
   parent=SizeBox_1
   widget=StockRemaining
   properties {
    text=Stock: 1
    textTransformPolicy=ToUpper
    font {
     bForceMonospaced=false
     bMaterialIsStencil=false
     fontMaterial=null
     fontObject=&BurbankSmall
     letterSpacing=0
     monospacedWidth=1
     size=20
     skewAmount=0.3
     typefaceFontName=Black
     outlineSettings {
      bApplyOutlineToDropShadows=false
      bMiteredCorners=false
      bSeparateFillAlpha=false
      outlineMaterial=null
      outlineSize=0
      outlineColor=<rgba 0 0 0 1>
     }
    }
    shadowOffset=<0 0>
   }
   slot {class=SizeBoxSlot, properties {verticalAlignment=VAlign_Bottom, padding=<ltrb 0 0 4 0>}}
  }
  CannotBuy {
   widgetClassPath=&Overlay
   bIsVariable=true
   parent=StateVisibilitySwitcher
   widget=CannotBuy
   slot {class=CommonVisibilitySwitcherSlot}
  }
  HorizontalBox_1 {
   widgetClassPath=&HorizontalBox
   parent=CannotBuy
   widget=HorizontalBox_1
   slot {class=OverlaySlot, properties {horizontalAlignment=HAlign_Fill, verticalAlignment=VAlign_Center, padding=<ltrb 8 8 12 8>}}
  }
  ErrorIcon {
   widgetClassPath=&Image
   parent=HorizontalBox_1
   widget=ErrorIcon
   properties {brush {
     bIsDynamicallyLoaded=false
     drawAs=Image
     imageType=NoImage
     mirroring=NoMirror
     resourceName=None
     resourceObject=/Game/UI/Foundation/Textures/Buttons/T-SlotIcon-NotRecyclable-64
     tiling=NoTile
     imageSize=<32 32>
     margin=<ltrb 0 0 0 0>
     outlineSettings {
      bUseBrushTransparency=false
      roundingType=HalfHeightRadius
      width=0
      color {colorUseRule=UseColor_Specified, specifiedColor=<rgba 0 0 0 0>}
      cornerRadii=<0 0 0 0>
     }
     tintColor {colorUseRule=UseColor_Specified, specifiedColor=<rgba 1 0.0578 0.0666 1>}
     uVRegion {bIsValid=false, max=<0 0>, min=<0 0>}
    }}
   slot {class=HorizontalBoxSlot}
  }
  ErrorText {
   widgetClassPath=&CommonTextBlock
   bIsVariable=true
   parent=HorizontalBox_1
   widget=ErrorText
   properties {
    justification=Center
    text=Out of stock
    textTransformPolicy=ToUpper
    font {
     bForceMonospaced=false
     bMaterialIsStencil=false
     fontMaterial=null
     fontObject=&BurbankSmall
     letterSpacing=0
     monospacedWidth=1
     size=16
     skewAmount=0
     typefaceFontName=Black
     outlineSettings {
      bApplyOutlineToDropShadows=false
      bMiteredCorners=false
      bSeparateFillAlpha=true
      outlineMaterial=&M-FontShadow
      outlineSize=2
      outlineColor=<rgba 0 0 0 0.5>
     }
    }
    shadowOffset=<0 0>
   }
   slot {class=HorizontalBoxSlot, properties {padding=<ltrb 5 0 0 0>}}
  }
 ]
}
